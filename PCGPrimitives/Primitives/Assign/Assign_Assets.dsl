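; PCG primitive graph that assigns assets to its input, with controls for transform offsets, seeded random ranges, bounds matching, and attribute-driven mesh or index selection.
; It tags the input "Input:0", then chains inputs-management, core-process and prepare-output subgraphs.
; Graph parameters expose input processing, fallback, core and output stages as swappable subgraph overrides.
; The Assets array is declared but not explicitly passed to any subgraph here.
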(graph Assign_Assets
  (params (InputsProcessingGraph :type (SoftObject /Script/PCG.PCGGraphInterface) :default /PCGPrimitives/Primitives/_Template_/Primitives_InputsProcessing.Primitives_InputsProcessing) (InputsFallbackGraph :type (SoftObject /Script/PCG.PCGGraphInterface) :default /PCGPrimitives/Primitives/Assign/Subgraphs/Assign_Assets_InputsFallback.Assign_Assets_InputsFallback) (CoreProcessGraph :type (SoftObject /Script/PCG.PCGGraphInterface) :default /PCGPrimitives/Primitives/Assign/Subgraphs/Assign_Assets_CoreProcess.Assign_Assets_CoreProcess) (PrepareOutputsGraph :type (SoftObject /Script/PCG.PCGGraphInterface) :default /PCGPrimitives/Primitives/_Template_/Primitives_PrepareOutputs.Primitives_PrepareOutputs) (Assets :type (Array (SoftObject /Script/CoreUObject.Object))) (PositionOffset :type Vector) (RotationOffset :type Rotator) (ScaleOffset :type Vector :default (1.0 1.0 1.0)) (ExtentsScale :type Vector :default (1.0 1.0 1.0)) (UseMeshBounds :type Bool :default true) (MatchScaleToInputBounds :type Bool) (UseMeshAttribute :type Bool) (MeshAttribute :type String) (RandomPositionRange :type Vector) (RandomRotationRange :type Vector) (RandomScaleRange :type Vector) (Seed :type Int) (UniformScale :type Bool :default true) (UseIndexattribute :type Bool) (IndexAttribute :type String))
  (inputs (PrimaryInput :type Any))
  (bind AddTags_0 (Add%20Tags :TagsToAdd "Input:0"))
  ;[0 0 0 0.5 32] Inputs
  (bind Subgraph_0 (Subgraph :Graph /PCGPrimitives/Primitives/_Template_/Subgraphs/Primitives_InputsManagement.Primitives_InputsManagement :InputsFallbackGraph InputsFallbackGraph :InputsProcessingGraph InputsProcessingGraph))
  ;[0 0 0 0.5 32] Core Process
  (bind Subgraph_3 (Subgraph :Graph /PCGPrimitives/Primitives/Assign/Subgraphs/Assign_Assets_CoreProcess.Assign_Assets_CoreProcess :ExtentsScale ExtentsScale :Index_attribute IndexAttribute :MatchScaleToInputBounds MatchScaleToInputBounds :MeshAttrinute MeshAttribute :PositionOffset PositionOffset :RandomPositionRange RandomPositionRange :RandomRotationRange RandomRotationRange :RandomScaleRange RandomScaleRange :RotationOffset RotationOffset :ScaleOffset ScaleOffset :Seed-param Seed :SubgraphOverride-param CoreProcessGraph :uniform_scale UniformScale :Use_Index_attribute UseIndexattribute :UseMeshAttribute UseMeshAttribute :UseMeshBounds UseMeshBounds))
  ;[0 0 0 0.5 32] Prepare Output
  (bind Subgraph_4 (Subgraph :Graph /PCGPrimitives/Primitives/_Template_/Primitives_PrepareOutputs.Primitives_PrepareOutputs :SubgraphOverride-param PrepareOutputsGraph))
  (connect DefaultInputNode PrimaryInput AddTags_0 In)
  (connect AddTags_0 Out Subgraph_0 In)
  (connect Subgraph_0 Out Subgraph_3 In)
  (connect Subgraph_3 Out Subgraph_4 In)
  (output Out Subgraph_4))
