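; Assign_Rule is a PCG primitive graph that assigns mesh modules by a grammar string, using spline-forced input processing by default.
; It tags the input, then chains inputs management, core process and output preparation subgraphs, returning points.
; The core receives grammar, module definitions, random transform ranges, offset, scale, mesh override and a spawn toggle.
; Graph parameters make input processing, fallback, core and output stages swappable, template-style.
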(graph Assign_Rule
  (params (InputsProcessingGraph :type (SoftObject /Script/PCG.PCGGraphInterface) :default /PCGPrimitives/Primitives/Shared/InputsProcessing/InputsProcessing_ForceSpline.InputsProcessing_ForceSpline) (InputsFallbackGraph :type (SoftObject /Script/PCG.PCGGraphInterface) :default /PCGPrimitives/Primitives/_Template_/Primitives_InputsFallback.Primitives_InputsFallback) (CoreProcessGraph :type (SoftObject /Script/PCG.PCGGraphInterface) :default /PCGPrimitives/Primitives/Assign/Subgraphs/Assign_Rule_CoreProcess.Assign_Rule_CoreProcess) (PrepareOutputsGraph :type (SoftObject /Script/PCG.PCGGraphInterface) :default /PCGPrimitives/Primitives/Assign/Subgraphs/Assign_Rule_PrepareOutputs.Assign_Rule_PrepareOutputs) (Grammar :type String :default "C*") (ModulesInfos :type (Array (Struct /PCGPrimitives/Grammar/Structure/STC_Module.STC_Module)) :default [(struct :Mesh /PCG/DebugObjects/PCG_Cube.PCG_Cube :Symbol "C" :DebugColor (struct :B 122 :G 227 :R 255 :A 255)) (struct)]) (RandPosRange :type Vector) (RandRotRange :type Vector) (RandScaleRange :type Vector) (PosOffset :type Vector) (OverrideScale :type Vector :default (1.0 1.0 1.0)) (MeshOverride :type (SoftObject /Script/Engine.StaticMesh)) (SpawnAssets :type Bool))
  (inputs (PrimaryInput :type Any))
  (bind AddTags_0 (Add%20Tags :TagsToAdd "Input:0"))
  ;[0 0 0 0.5 32] Inputs
  (bind Subgraph_0 (Subgraph :Graph /PCGPrimitives/Primitives/_Template_/Subgraphs/Primitives_InputsManagement.Primitives_InputsManagement :InputsFallbackGraph InputsFallbackGraph :InputsProcessingGraph InputsProcessingGraph))
  ;[0 0 0 0.5 32] Core Process
  (bind Subgraph_3 (Subgraph :Graph /PCGPrimitives/Primitives/Assign/Subgraphs/Assign_Rule_CoreProcess.Assign_Rule_CoreProcess :Grammar Grammar :MeshOverride MeshOverride :Modules_infos ModulesInfos :OverrideScale OverrideScale :PosOffset PosOffset :RandPosRange RandPosRange :RandRotRange RandRotRange :RandScaleRange RandScaleRange :SpawnAssets SpawnAssets :SubgraphOverride-param CoreProcessGraph))
  ;[0 0 0 0.5 32] Prepare Output
  (bind Subgraph_4 (Subgraph :Graph /PCGPrimitives/Primitives/Assign/Subgraphs/Assign_ShapeGrammarDefinition_PrepareOutputs.Assign_ShapeGrammarDefinition_PrepareOutputs :SubgraphOverride-param PrepareOutputsGraph))
  (connect DefaultInputNode PrimaryInput AddTags_0 In)
  (connect AddTags_0 Out Subgraph_0 In)
  (connect Subgraph_0 Out Subgraph_3 In)
  (connect Subgraph_3 Out Subgraph_4 In)
  (output Out Subgraph_4.Points))
