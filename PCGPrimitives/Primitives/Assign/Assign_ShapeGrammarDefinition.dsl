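; PCG primitive graph that assigns a shape grammar definition to its input, outputting Footprints and Points.
; It tags the input, then chains input management (spline-forcing processing plus fallback by default), core process and output preparation subgraphs.
; The core process receives the definition, a size or size attribute name, and an asset-spawning toggle.
; Graph parameters expose each stage's processing graph, so the pipeline can be swapped without editing it.
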(graph Assign_ShapeGrammarDefinition
  (params (InputsProcessingGraph :type (SoftObject /Script/PCG.PCGGraphInterface) :default /PCGPrimitives/Primitives/Shared/InputsProcessing/InputsProcessing_ForceSpline.InputsProcessing_ForceSpline) (InputsFallbackGraph :type (SoftObject /Script/PCG.PCGGraphInterface) :default /PCGPrimitives/Primitives/_Template_/Primitives_InputsFallback.Primitives_InputsFallback) (CoreProcessGraph :type (SoftObject /Script/PCG.PCGGraphInterface) :default /PCGPrimitives/Primitives/Assign/Subgraphs/Assign_ShapeGrammarDefinition_CoreProcess.Assign_ShapeGrammarDefinition_CoreProcess) (PrepareOutputsGraph :type (SoftObject /Script/PCG.PCGGraphInterface) :default /PCGPrimitives/Primitives/Assign/Subgraphs/Assign_ShapeGrammarDefinition_PrepareOutputs.Assign_ShapeGrammarDefinition_PrepareOutputs) (ShapeGrammarDefinition :type (SoftObject /PCGPrimitives/Grammar/BP/BP_ShapeGrammarDefinition.BP_ShapeGrammarDefinition_C)) (Size :type Double :default 5e+03) (UseAttributeForSize :type String) (SpawnAssets :type Bool :default true))
  (inputs (PrimaryInput :type Any))
  (outputs (Footprints :type Any) (Points :type Any))
  (bind AddTags_0 (Add%20Tags :TagsToAdd "Input:0"))
  ;[0 0 0 0.5 32] Inputs
  (bind Subgraph_0 (Subgraph :Graph /PCGPrimitives/Primitives/_Template_/Subgraphs/Primitives_InputsManagement.Primitives_InputsManagement :InputsFallbackGraph InputsFallbackGraph :InputsProcessingGraph InputsProcessingGraph))
  ;[0 0 0 0.5 32] Core Process
  (bind Subgraph_3 (Subgraph :Graph /PCGPrimitives/Primitives/Assign/Subgraphs/Assign_ShapeGrammarDefinition_CoreProcess.Assign_ShapeGrammarDefinition_CoreProcess :ShapeGrammarDefinition ShapeGrammarDefinition :Size Size :SpawnAssets SpawnAssets :SubgraphOverride-param CoreProcessGraph :UseAttributeForSize UseAttributeForSize))
  ;[0 0 0 0.5 32] Prepare Output
  (bind Subgraph_4 (Subgraph :Graph /PCGPrimitives/Primitives/Assign/Subgraphs/Assign_ShapeGrammarDefinition_PrepareOutputs.Assign_ShapeGrammarDefinition_PrepareOutputs :SubgraphOverride-param PrepareOutputsGraph))
  (connect DefaultInputNode PrimaryInput AddTags_0 In)
  (connect AddTags_0 Out Subgraph_0 In)
  (connect Subgraph_0 Out Subgraph_3 In)
  (connect Subgraph_3 Out Subgraph_4 In)
  (output Footprints Subgraph_4.Footprints)
  (output Points Subgraph_4.Points))
